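# Sound cue for the FreshWave pickaxe hitting an enemy, using the shared pickaxe impact attenuation, property sheet and impacts sound class.
# After a short fixed delay, a weighted random node favors a lone impact from six variations over a mixed branch.
# The mixed branch pairs an impact from the same six waves with a random "Lyr" layer, whose weighted empty slot sometimes leaves the impact unlayered.
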SoundCue /BRCosmetics/Sounds/PickAxes/FreshWave/PickaxeImpactEnemy_FreshWave_Cue
root=Delay_0
nodes [
 {id=Delay_0, children [Random_4], properties {delayMax=0.1, delayMin=0.1}}
 {id=Mixer_0, children [Random_1,null,Random_6], properties {inputVolume [1,1,1]}}
 {id=Random_0, children [WavePlayer_0,WavePlayer_1,WavePlayer_2,WavePlayer_3,WavePlayer_4,WavePlayer_5], properties {weights [1,1,1,1,1,1]}}
 {
  id=Random_1
  children [WavePlayer_18, WavePlayer_19, WavePlayer_20, WavePlayer_21, WavePlayer_22, WavePlayer_23, null]
  properties {bRandomizeWithoutReplacement=false, weights [0.6,0.6,0.6,0.6,0.6,0.6,1.5]}
 }
 {id=Random_4, children [Random_0,Mixer_0], properties {weights [1,0.5]}}
 {
  id=Random_6
  children [WavePlayer_12, WavePlayer_13, WavePlayer_14, WavePlayer_15, WavePlayer_16, WavePlayer_17]
  properties {weights [1,1,1,1,1,1]}
 }
 {id=WavePlayer_0, properties {soundWaveAssetPtr=&/IP_Pickaxe_FreshWave_Impact_01}}
 {id=WavePlayer_1, properties {soundWaveAssetPtr=&/IP_Pickaxe_FreshWave_Impact_06}}
 {id=WavePlayer_12, properties {soundWaveAssetPtr=&IP_Pickaxe_FreshWave_Impact_01}}
 {id=WavePlayer_13, properties {soundWaveAssetPtr=&IP_Pickaxe_FreshWave_Impact_06}}
 {id=WavePlayer_14, properties {soundWaveAssetPtr=&/IP_Pickaxe_FreshWave_Impact_05}}
 {id=WavePlayer_15, properties {soundWaveAssetPtr=&/IP_Pickaxe_FreshWave_Impact_04}}
 {id=WavePlayer_16, properties {soundWaveAssetPtr=&/IP_Pickaxe_FreshWave_Impact_03}}
 {id=WavePlayer_17, properties {soundWaveAssetPtr=&/IP_Pickaxe_FreshWave_Impact_02}}
 {id=WavePlayer_18, properties {soundWaveAssetPtr=&/IP_Pickaxe_FreshWave_Impact_01_Lyr}}
 {id=WavePlayer_19, properties {soundWaveAssetPtr=&/IP_Pickaxe_FreshWave_Impact_02_Lyr}}
 {id=WavePlayer_2, properties {soundWaveAssetPtr=&IP_Pickaxe_FreshWave_Impact_05}}
 {id=WavePlayer_20, properties {soundWaveAssetPtr=&/IP_Pickaxe_FreshWave_Impact_03_Lyr}}
 {id=WavePlayer_21, properties {soundWaveAssetPtr=&/IP_Pickaxe_FreshWave_Impact_04_Lyr}}
 {id=WavePlayer_22, properties {soundWaveAssetPtr=&/IP_Pickaxe_FreshWave_Impact_05_Lyr}}
 {id=WavePlayer_23, properties {soundWaveAssetPtr=&/IP_Pickaxe_FreshWave_Impact_06_Lyr}}
 {id=WavePlayer_3, properties {soundWaveAssetPtr=&IP_Pickaxe_FreshWave_Impact_04}}
 {id=WavePlayer_4, properties {soundWaveAssetPtr=&IP_Pickaxe_FreshWave_Impact_03}}
 {id=WavePlayer_5, properties {soundWaveAssetPtr=&IP_Pickaxe_FreshWave_Impact_02}}
]
properties {
 attenuationSettings=/Game/Sounds/Attenuation/Fort_Pickaxe_Impact_Att
 audioPropertiesSheet=/BRCosmetics/Sounds/Settings/PropertySheets/Pickaxe/APS_Cosmetic_Pickaxe_Impact
 maxDistance=4300
 soundClassObject=&Fort_Impacts
 concurrencySet [null]
}
